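# Editor utility widget for the Sparks stage show that provides buttons to convert Mojo stage show assets.
# A stacked layout pairs labels with five buttons covering gameplay, MIDI and FX manager sequences, plus gameplay and MIDI color palettes.
# Each button's click event makes a single call to its matching Mojo conversion function, whose logic is not defined in this graph.
# The widget holds no variables, and its construct and tick events are empty.
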EditorUtilityWidgetBlueprint /SparksStageShow/ConvertStageShowAssetsEditorUtility
graphs [EventGraph {nodes [
   "On Clicked (ConvertMojoStageMapGameplaySequence)"() {is=title K2Node_ComponentBoundEvent}
   > ConvertMojoStageMapGameplaySequence!(execute) -> (then)
   "On Clicked (ConvertMojoStageMapMidiSequence)"() {is=title K2Node_ComponentBoundEvent}
   > ConvertMojoStageMapMidiSequence!(execute) -> (then)
   "On Clicked (ConvertMojoFXManagerSequences)"() {is=title K2Node_ComponentBoundEvent}
   > ConvertMojoFxManagerSequences!(execute) -> (then)
   "On Clicked (ConvertGameplayColorPalettes)"() {is=title K2Node_ComponentBoundEvent}
   > ConvertMojoStageEventGameplayColorPalettes!(execute) -> (then)
   "On Clicked (ConvertMIDIColorPalettes)"() {is=title K2Node_ComponentBoundEvent}
   > ConvertMojoStageEventMIDIColorPalettes!(execute) -> (then)
   Event PreConstruct()
   Event Construct()
   Event Tick()
  ]}]
variables []
widgets {
 info {rootWidgetClass=&StackBox, widgetCount=11}
 widgets [
  StackBox_0 {widgetClassPath=&StackBox, widget=StackBox_0}
  GamePlayMap {
   widgetClassPath=&TextBlock
   parent=StackBox_0
   widget=GamePlayMap
   properties {text=Convert Gameplay Sequences}
   slot {class=StackBoxSlot, properties {horizontalAlignment=HAlign_Left, verticalAlignment=VAlign_Center}}
  }
  ConvertMojoStageMapGameplaySequence {
   widgetClassPath=&EditorUtilityButton
   parent=StackBox_0
   widget=ConvertMojoStageMapGameplaySequence
   slot {class=StackBoxSlot, properties {horizontalAlignment=HAlign_Center, verticalAlignment=VAlign_Center}}
  }
  ConvertMidiMap {
   widgetClassPath=&TextBlock
   parent=StackBox_0
   widget=ConvertMidiMap
   properties {text=Convert Midi Sequences}
   slot {class=StackBoxSlot, properties {horizontalAlignment=HAlign_Left, verticalAlignment=VAlign_Center}}
  }
  ConvertMojoStageMapMidiSequence {
   widgetClassPath=&EditorUtilityButton
   parent=StackBox_0
   widget=ConvertMojoStageMapMidiSequence
   slot {class=StackBoxSlot, properties {horizontalAlignment=HAlign_Center, verticalAlignment=VAlign_Center}}
  }
  ConvertMojoManaager {
   widgetClassPath=&TextBlock
   parent=StackBox_0
   widget=ConvertMojoManaager
   properties {text=Convert Sequences}
   slot {class=StackBoxSlot, properties {horizontalAlignment=HAlign_Left, verticalAlignment=VAlign_Center}}
  }
  ConvertMojoFXManagerSequences {
   widgetClassPath=&EditorUtilityButton
   parent=StackBox_0
   widget=ConvertMojoFXManagerSequences
   slot {class=StackBoxSlot, properties {horizontalAlignment=HAlign_Center, verticalAlignment=VAlign_Center}}
  }
  ConverGameplayColorPalettes {
   widgetClassPath=&TextBlock
   parent=StackBox_0
   widget=ConverGameplayColorPalettes
   properties {text=Convert Gameplay Color Palettes}
   slot {class=StackBoxSlot, properties {horizontalAlignment=HAlign_Left, verticalAlignment=VAlign_Center}}
  }
  ConvertGameplayColorPalettes {
   widgetClassPath=&EditorUtilityButton
   parent=StackBox_0
   widget=ConvertGameplayColorPalettes
   slot {class=StackBoxSlot, properties {horizontalAlignment=HAlign_Center, verticalAlignment=VAlign_Center}}
  }
  ConverMIDIColorPalettes {
   widgetClassPath=&TextBlock
   parent=StackBox_0
   widget=ConverMIDIColorPalettes
   properties {text=Convert MIDI Color Palettes}
   slot {class=StackBoxSlot, properties {horizontalAlignment=HAlign_Left, verticalAlignment=VAlign_Center}}
  }
  ConvertMIDIColorPalettes {
   widgetClassPath=&EditorUtilityButton
   parent=StackBox_0
   widget=ConvertMIDIColorPalettes
   slot {class=StackBoxSlot, properties {horizontalAlignment=HAlign_Center, verticalAlignment=VAlign_Center}}
  }
 ]
}
